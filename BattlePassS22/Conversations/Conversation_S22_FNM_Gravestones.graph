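# Conversation database for the Season 22 FNM gravestones, supplying ten joke epitaphs shown through the NPC conversation system.
# Each gravestone has its own entry tag, Conversation.Entry.FNM.GS.01 to 10, leading to a single Speech node with the inscription.
# Each Speech feeds a Continue node with a back icon and a lone Exit choice, so the chains are independent, with no branching or further outputs.
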ConversationDatabase /BattlePassS22/Conversations/Conversation_S22_FNM_Gravestones
entry_points [
 {tag=Conversation.Entry.FNM.GS.01, destinations [#1]}
 {tag=Conversation.Entry.FNM.GS.02, destinations [#2]}
 {tag=Conversation.Entry.FNM.GS.03, destinations [#3]}
 {tag=Conversation.Entry.FNM.GS.04, destinations [#4]}
 {tag=Conversation.Entry.FNM.GS.05, destinations [#5]}
 {tag=Conversation.Entry.FNM.GS.06, destinations [#6]}
 {tag=Conversation.Entry.FNM.GS.07, destinations [#7]}
 {tag=Conversation.Entry.FNM.GS.08, destinations [#8]}
 {tag=Conversation.Entry.FNM.GS.09, destinations [#9]}
 {tag=Conversation.Entry.FNM.GS.10, destinations [#10]}
]
nodes [
 ConversationEntryPointNode_10 {
  id=#10
  class=ConversationEntryPointNode
  outputs [#11]
  properties {entryTag {tagName=Conversation.Entry.FNM.GS.10}}
 }
 Speech_C_14 {
  id=#11
  class=Speech_C
  outputs [#12]
  properties {nodeName="", generalConfig {defaultMessage="In case of zombie outbreak, this grave will be unoccupied"}}
 }
 Continue_C_14 {
  id=#12
  class=Continue_C
  properties {softTaskIcon=&T_UI_NPC_Back}
  sub_nodes [ConversationChoiceNode_16 {class=ConversationChoiceNode, properties {defaultChoiceDisplayText=Exit, nodeName=Exit}}]
 }
 ConversationEntryPointNode_9 {
  id=#9
  class=ConversationEntryPointNode
  outputs [#13]
  properties {entryTag {tagName=Conversation.Entry.FNM.GS.09}}
 }
 Speech_C_13 {
  id=#13
  class=Speech_C
  outputs [#14]
  properties {nodeName="", generalConfig {defaultMessage=Doctor Slone}}
 }
 Continue_C_13 {
  id=#14
  class=Continue_C
  properties {softTaskIcon=&T_UI_NPC_Back}
  sub_nodes [ConversationChoiceNode_15 {class=ConversationChoiceNode, properties {defaultChoiceDisplayText=Exit, nodeName=Exit}}]
 }
 ConversationEntryPointNode_8 {
  id=#8
  class=ConversationEntryPointNode
  outputs [#15]
  properties {entryTag {tagName=Conversation.Entry.FNM.GS.08}}
 }
 Speech_C_12 {
  id=#15
  class=Speech_C
  outputs [#16]
  properties {nodeName="", generalConfig {defaultMessage=This is NOT a secret apocalypse bunker.}}
 }
 Continue_C_12 {
  id=#16
  class=Continue_C
  properties {softTaskIcon=&T_UI_NPC_Back}
  sub_nodes [ConversationChoiceNode_14 {class=ConversationChoiceNode, properties {defaultChoiceDisplayText=Exit, nodeName=Exit}}]
 }
 ConversationEntryPointNode_7 {
  id=#7
  class=ConversationEntryPointNode
  outputs [#17]
  properties {entryTag {tagName=Conversation.Entry.FNM.GS.07}}
 }
 Speech_C_11 {
  id=#17
  class=Speech_C
  outputs [#18]
  properties {nodeName="", generalConfig {defaultMessage="Willow's Axe Cache - Don't Touch!"}}
 }
 Continue_C_11 {
  id=#18
  class=Continue_C
  properties {softTaskIcon=&T_UI_NPC_Back}
  sub_nodes [ConversationChoiceNode_13 {class=ConversationChoiceNode, properties {defaultChoiceDisplayText=Exit, nodeName=Exit}}]
 }
 ConversationEntryPointNode_6 {
  id=#6
  class=ConversationEntryPointNode
  outputs [#19]
  properties {entryTag {tagName=Conversation.Entry.FNM.GS.06}}
 }
 Speech_C_10 {
  id=#19
  class=Speech_C
  outputs [#20]
  properties {nodeName="", generalConfig {defaultMessage=This plot is sponsored by No Sweat Insurance}}
 }
 Continue_C_10 {
  id=#20
  class=Continue_C
  properties {softTaskIcon=&T_UI_NPC_Back}
  sub_nodes [ConversationChoiceNode_12 {class=ConversationChoiceNode, properties {defaultChoiceDisplayText=Exit, nodeName=Exit}}]
 }
 ConversationEntryPointNode_5 {
  id=#5
  class=ConversationEntryPointNode
  outputs [#21]
  properties {entryTag {tagName=Conversation.Entry.FNM.GS.05}}
 }
 Speech_C_9 {
  id=#21
  class=Speech_C
  outputs [#22]
  properties {nodeName="", generalConfig {defaultMessage=IYKYK}}
 }
 Continue_C_9 {
  id=#22
  class=Continue_C
  properties {softTaskIcon=&T_UI_NPC_Back}
  sub_nodes [ConversationChoiceNode_11 {class=ConversationChoiceNode, properties {defaultChoiceDisplayText=Exit, nodeName=Exit}}]
 }
 ConversationEntryPointNode_4 {
  id=#4
  class=ConversationEntryPointNode
  outputs [#23]
  properties {entryTag {tagName=Conversation.Entry.FNM.GS.04}}
 }
 Speech_C_8 {
  id=#23
  class=Speech_C
  outputs [#24]
  properties {nodeName="", generalConfig {defaultMessage="Meowscles (Life 3 of 9)"}}
 }
 Continue_C_8 {
  id=#24
  class=Continue_C
  properties {softTaskIcon=&T_UI_NPC_Back}
  sub_nodes [ConversationChoiceNode_10 {class=ConversationChoiceNode, properties {defaultChoiceDisplayText=Exit, nodeName=Exit}}]
 }
 ConversationEntryPointNode_2 {
  id=#3
  class=ConversationEntryPointNode
  outputs [#25]
  properties {entryTag {tagName=Conversation.Entry.FNM.GS.03}}
 }
 Speech_C_7 {
  id=#25
  class=Speech_C
  outputs [#26]
  properties {nodeName="", generalConfig {defaultMessage=Cuddle Team Founder}}
 }
 Continue_C_7 {
  id=#26
  class=Continue_C
  properties {softTaskIcon=&T_UI_NPC_Back}
  sub_nodes [ConversationChoiceNode_9 {class=ConversationChoiceNode, properties {defaultChoiceDisplayText=Exit, nodeName=Exit}}]
 }
 ConversationEntryPointNode_1 {
  id=#2
  class=ConversationEntryPointNode
  outputs [#27]
  properties {entryTag {tagName=Conversation.Entry.FNM.GS.02}}
 }
 Speech_C_5 {
  id=#27
  class=Speech_C
  outputs [#28]
  properties {nodeName="", generalConfig {defaultMessage=Available for Sublet - contact Grisabelle}}
 }
 Continue_C_1 {
  id=#28
  class=Continue_C
  properties {softTaskIcon=&T_UI_NPC_Back}
  sub_nodes [ConversationChoiceNode_0 {class=ConversationChoiceNode, properties {defaultChoiceDisplayText=Exit, nodeName=Exit}}]
 }
 ConversationEntryPointNode_0 {
  id=#1
  class=ConversationEntryPointNode
  outputs [#29]
  properties {entryTag {tagName=Conversation.Entry.FNM.GS.01}}
 }
 Speech_C_6 {
  id=#29
  class=Speech_C
  outputs [#30]
  properties {nodeName="", generalConfig {defaultMessage=Didn't thank the Bus Driver}}
 }
 Continue_C_6 {
  id=#30
  class=Continue_C
  properties {softTaskIcon=&T_UI_NPC_Back}
  sub_nodes [ConversationChoiceNode_8 {class=ConversationChoiceNode, properties {defaultChoiceDisplayText=Exit, nodeName=Exit}}]
 }
]
speakers [{}, {participant=Conversation.Participant.NPC}]
