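# Conversation database for a Season 22 narrative quest in which the player processes a root sound through a step-by-step dialogue at a quest prop.
# A tagged entry point opens a linear chain of speeches that act as a percentage progress readout, each later step offered as a single audio-processing choice.
# Every step sends a numbered message side effect to the quest prop actor, and the final step also fires conversation success.
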ConversationDatabase /BattlePassS22/Conversations/Conversation_S22_P01_Q05_ProcessRootSound
entry_points [{tag=Conversation.Entry.Season22.Narrative.ProcessRootSound, destinations [#1]}]
nodes [
 ConversationEntryPointNode_7 {
  id=#1
  class=ConversationEntryPointNode
  outputs [#2]
  properties {entryTag {tagName=Conversation.Entry.Season22.Narrative.ProcessRootSound}}
 }
 Speech_C_12 {
  id=#2
  class=Speech_C
  outputs [#3]
  properties {generalConfig {defaultMessage=0%}}
  sub_nodes [ConvoSideEffect_QuestPropActor_Message_C_4 {class=ConvoSideEffect_QuestPropActor_Message_C}]
 }
 Speech_C_5 {
  id=#3
  class=Speech_C
  outputs [#4]
  properties {softTaskIcon=/Game/UI/Foundation/Textures/Icons/Replay/T-Icon-Replay-Forward-L, generalConfig {defaultMessage=25%}}
  sub_nodes [
   ConversationChoiceNode_6 {class=ConversationChoiceNode, properties {defaultChoiceDisplayText=Adjust Amplitudes}}
   ConvoSideEffect_QuestPropActor_Message_C_3 {class=ConvoSideEffect_QuestPropActor_Message_C, properties {message Number=1}}
  ]
 }
 Speech_C_6 {
  id=#4
  class=Speech_C
  outputs [#5]
  properties {softTaskIcon=&T-Icon-Replay-Forward-L, generalConfig {defaultMessage=50%}}
  sub_nodes [
   ConversationChoiceNode_2 {class=ConversationChoiceNode, properties {defaultChoiceDisplayText=Find Frequencies}}
   ConvoSideEffect_QuestPropActor_Message_C_2 {class=ConvoSideEffect_QuestPropActor_Message_C, properties {message Number=2}}
  ]
 }
 Speech_C_7 {
  id=#5
  class=Speech_C
  outputs [#6]
  properties {softTaskIcon=&T-Icon-Replay-Forward-L, generalConfig {defaultMessage=75%}}
  sub_nodes [
   ConversationChoiceNode_26 {class=ConversationChoiceNode, properties {defaultChoiceDisplayText=Nullify Noise}}
   ConvoSideEffect_QuestPropActor_Message_C_1 {class=ConvoSideEffect_QuestPropActor_Message_C, properties {message Number=3}}
  ]
 }
 Speech_C_17 {
  id=#6
  class=Speech_C
  outputs []
  properties {softTaskIcon=&T-Icon-Replay-Forward-L, generalConfig {defaultMessage=100%}}
  sub_nodes [
   ConversationChoiceNode_8 {class=ConversationChoiceNode, properties {defaultChoiceDisplayText=Decode Dialogue}}
   ConvoSideEffect_QuestPropActor_Message_C_5 {class=ConvoSideEffect_QuestPropActor_Message_C, properties {message Number=4}}
   ConvoSideEffect_QuestPropActor_ConvoSuccess_C_0 {class=ConvoSideEffect_QuestPropActor_ConvoSuccess_C}
  ]
 }
]
