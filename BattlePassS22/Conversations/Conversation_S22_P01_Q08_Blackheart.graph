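# Conversation database for Blackheart's dialogue in Season 22 narrative quest Q08 (P01), about the Scientist's research notes.
# The Q08 entry point leads to an auto-selected main-quest choice, shown only while the quest is active, the quest and its first objective are incomplete, and the NPC carries the Blackheart entry tag.
# Three chained speech lines then end in a task that completes that objective by backend name.
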ConversationDatabase /BattlePassS22/Conversations/Conversation_S22_P01_Q08_Blackheart
entry_points [{tag=Conversation.Entry.Season22.Narrative.Q08, destinations [#1]}]
nodes [
 ConversationEntryPointNode_0 {
  id=#1
  class=ConversationEntryPointNode
  outputs [#2]
  properties {entryTag {tagName=Conversation.Entry.Season22.Narrative.Q08}}
 }
 Speech_C_32 {
  id=#2
  class=Speech_C
  outputs [#3]
  properties {
   softTaskIcon=&T_UI_NPC_MainQuest
   generalConfig {defaultMessage="Yeah, we \"found\" the Scientist's research notes."}
  }
  sub_nodes [
   ConversationChoiceNode_43 {
    class=ConversationChoiceNode
    properties {
     defaultChoiceDisplayText=Ask about The Scientist's notes
     choiceTags {gameplayTags [{tagName=Conversation.Choice.AutoSelectIfValid}, {tagName=Conversation.Choice.Category.MainQuestline}]}
    }
   }
   FortConversationRequirement_ControllerRequirement_16 {
    class=FortConversationRequirement_ControllerRequirement
    properties {
     nodeName=Require: Has Completed Quest
     requirement {
      class=FortControllerRequirement_HasCompletedQuest
      properties {bInverseCondition=true, questDefinition=/BattlePassS22/Items/QuestItems/Narrative/quest_s22_narrative_p01_q08}
     }
    }
   }
   FortConversationRequirement_ControllerRequirement_17 {
    class=FortConversationRequirement_ControllerRequirement
    properties {
     nodeName=Require: Is Quest Active
     requirement {
      class=FortControllerRequirement_IsQuestActive
      properties {questDefinition=&quest_s22_narrative_p01_q08}
     }
    }
   }
   FortConversationRequirement_ControllerRequirement_42 {
    class=FortConversationRequirement_ControllerRequirement
    properties {
     nodeName=Require: Has Completed Quest Objective
     requirement {
      class=FortControllerRequirement_HasCompletedQuestObjective
      properties {
       bInverseCondition=true
       objectiveBackendName=quest_s22_narrative_p01_q08_obj0
       questDefinition=&quest_s22_narrative_p01_q08
      }
     }
    }
   }
   FortConversationRequirement_ControllerRequirement_19 {
    class=FortConversationRequirement_ControllerRequirement
    properties {
     nodeName=Require: NPC Has Conversation Tag
     participantId {tagName=Conversation.Participant.NPC}
     requirement {class=FortControllerRequirement_NPCHasConversationEntryTag, properties {queryTag {tagName=Conversation.Entry.Blackheart}}}
    }
   }
  ]
 }
 Speech_C_28 {
  id=#3
  class=Speech_C
  outputs [#4]
  properties {generalConfig {defaultMessage="Couldn't sell 'em, so we tossed 'em."}}
  sub_nodes [ConversationChoiceNode_34 {class=ConversationChoiceNode, properties {defaultChoiceDisplayText=Continue}}]
 }
 Speech_C_33 {
  id=#4
  class=Speech_C
  outputs [#5]
  properties {generalConfig {defaultMessage="You wanna root through our trash? Help yerself. "}}
  sub_nodes [ConversationChoiceNode_44 {class=ConversationChoiceNode, properties {defaultChoiceDisplayText=Continue}}]
 }
 FortConversationTaskNode_CompleteQuestObjectiveByBackendName_2 {
  id=#5
  class=FortConversationTaskNode_CompleteQuestObjectiveByBackendName
  properties {objectiveBackendName=quest_s22_narrative_p01_q08_obj0, questDefinition=&quest_s22_narrative_p01_q08}
 }
]
