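# Conversation database for the Bushranger NPC's dialogue step in Season 22 narrative quest Part 2, Q12.
# The Q12 entry point leads to an auto-selected main-questline speech, gated on the quest being active, quest and objective not yet completed, and the NPC carrying the Bushranger tag.
# A second speech about planting a seed follows on Continue, then a task node completes the quest's obj0 objective.
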ConversationDatabase /BattlePassS22/Conversations/Conversation_S22_P02_Q12_Bushy
entry_points [{tag=Conversation.Entry.Season22.Narrative.Q12, destinations [#1]}]
nodes [
 ConversationEntryPointNode_0 {
  id=#1
  class=ConversationEntryPointNode
  outputs [#2]
  properties {entryTag {tagName=Conversation.Entry.Season22.Narrative.Q12}}
 }
 Speech_C_3 {
  id=#2
  class=Speech_C
  outputs [#3]
  properties {softTaskIcon=&T_UI_NPC_MainQuest, generalConfig {defaultMessage="Yes! Talktalk treetalk!"}}
  sub_nodes [
   ConversationChoiceNode_3 {
    class=ConversationChoiceNode
    properties {
     defaultChoiceDisplayText=Talk to Bushranger
     choiceTags {gameplayTags [{tagName=Conversation.Choice.AutoSelectIfValid}, {tagName=Conversation.Choice.Category.MainQuestline}]}
    }
   }
   FortConversationRequirement_ControllerRequirement_12 {
    class=FortConversationRequirement_ControllerRequirement
    properties {
     nodeName=Require: Has Completed Quest
     requirement {
      class=FortControllerRequirement_HasCompletedQuest
      properties {bInverseCondition=true, questDefinition=/BattlePassS22/Items/QuestItems/Narrative/Part_02/quest_s22_narrative_p02_q12}
     }
    }
   }
   FortConversationRequirement_ControllerRequirement_13 {
    class=FortConversationRequirement_ControllerRequirement
    properties {
     nodeName=Require: Is Quest Active
     requirement {
      class=FortControllerRequirement_IsQuestActive
      properties {questDefinition=&quest_s22_narrative_p02_q12}
     }
    }
   }
   FortConversationRequirement_ControllerRequirement_14 {
    class=FortConversationRequirement_ControllerRequirement
    properties {
     nodeName=Require: Has Completed Quest Objective
     requirement {
      class=FortControllerRequirement_HasCompletedQuestObjective
      properties {
       bInverseCondition=true
       objectiveBackendName=quest_s22_narrative_p02_q12_obj0
       questDefinition=&quest_s22_narrative_p02_q12
      }
     }
    }
   }
   FortConversationRequirement_ControllerRequirement_16 {
    class=FortConversationRequirement_ControllerRequirement
    properties {
     nodeName=Require: NPC Has Conversation Tag
     participantId {tagName=Conversation.Participant.NPC}
     requirement {class=FortControllerRequirement_NPCHasConversationEntryTag, properties {queryTag {tagName=Conversation.Entry.Bushranger}}}
    }
   }
  ]
 }
 Speech_C_0 {
  id=#3
  class=Speech_C
  outputs [#4]
  properties {generalConfig {defaultMessage=Make seedplant. Then treetalk get LOUD.}}
  sub_nodes [ConversationChoiceNode_0 {class=ConversationChoiceNode, properties {defaultChoiceDisplayText=Continue}}]
 }
 FortConversationTaskNode_CompleteQuestObjectiveByBackendName_2 {
  id=#4
  class=FortConversationTaskNode_CompleteQuestObjectiveByBackendName
  properties {
   objectiveBackendName=quest_s22_narrative_p02_q12_obj0
   questDefinition=&quest_s22_narrative_p02_q12
  }
 }
]
