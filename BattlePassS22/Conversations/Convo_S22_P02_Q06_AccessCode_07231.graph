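# Season 22 battle pass quest conversation that acts as a keypad puzzle where the player enters access code 07231 on a quest prop.
# Each step branches into three digit choices: the correct one shows the code so far and continues, while either wrong one shows a denial and ends the conversation.
# QuestPropActor Message side effects pass a progress number for each step, zero on denial.
# The final correct digit reports the download and fires the ConvoSuccess side effect.
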ConversationDatabase /BattlePassS22/Conversations/Convo_S22_P02_Q06_AccessCode_07231
entry_points [{tag=Conversation.Entry.Season22.Narrative.Code.07231, destinations [#1]}]
nodes [
 ConversationEntryPointNode_7 {
  id=#1
  class=ConversationEntryPointNode
  outputs [#2]
  properties {entryTag {tagName=Conversation.Entry.Season22.Narrative.Code.07231}}
 }
 Speech_C_22 {
  id=#2
  class=Speech_C
  outputs [#3, #4, #5]
  properties {
   softTaskIcon=/Game/UI/Foundation/Textures/Icons/Replay/T-Icon-Replay-Plus-L
   generalConfig {defaultMessage=Enter Valid Access Code:, defaultMetadataParameters [Message number {value=0}]}
  }
  sub_nodes [ConvoSideEffect_QuestPropActor_Message_C_14 {class=ConvoSideEffect_QuestPropActor_Message_C}]
 }
 Speech_C_9 {
  id=#5
  class=Speech_C
  outputs []
  properties {
   softTaskIcon=&T-Icon-Replay-Plus-L
   generalConfig {defaultMessage=Denied: number entered was incorrect, defaultMetadataParameters [Message number {value=0}]}
  }
  sub_nodes [
   ConvoSideEffect_QuestPropActor_Message_C_12 {class=ConvoSideEffect_QuestPropActor_Message_C}
   ConversationChoiceNode_11 {class=ConversationChoiceNode, properties {defaultChoiceDisplayText=Enter 5}}
  ]
 }
 Speech_C_23 {
  id=#4
  class=Speech_C
  outputs [#6, #7, #8]
  properties {
   softTaskIcon=&T-Icon-Replay-Plus-L
   generalConfig {defaultMessage=0, defaultMetadataParameters [Message number {value=1}]}
  }
  sub_nodes [
   ConvoSideEffect_QuestPropActor_Message_C_15 {class=ConvoSideEffect_QuestPropActor_Message_C, properties {message Number=1}}
   ConversationChoiceNode_12 {class=ConversationChoiceNode, properties {defaultChoiceDisplayText=Enter 0}}
  ]
 }
 Speech_C_18 {
  id=#8
  class=Speech_C
  outputs [#9, #10, #11]
  properties {
   softTaskIcon=&T-Icon-Replay-Plus-L
   generalConfig {defaultMessage=07, defaultMetadataParameters [Message number {value=2}]}
  }
  sub_nodes [
   ConvoSideEffect_QuestPropActor_Message_C_13 {class=ConvoSideEffect_QuestPropActor_Message_C, properties {message Number=2}}
   ConversationChoiceNode_10 {class=ConversationChoiceNode, properties {defaultChoiceDisplayText=Enter 7}}
  ]
 }
 Speech_C_11 {
  id=#11
  class=Speech_C
  outputs []
  properties {
   softTaskIcon=&T-Icon-Replay-Plus-L
   generalConfig {defaultMessage=Denied: number entered was incorrect, defaultMetadataParameters [Message number {value=0}]}
  }
  sub_nodes [
   ConvoSideEffect_QuestPropActor_Message_C_1 {class=ConvoSideEffect_QuestPropActor_Message_C}
   ConversationChoiceNode_0 {class=ConversationChoiceNode, properties {defaultChoiceDisplayText=Enter 1}}
  ]
 }
 Speech_C_13 {
  id=#10
  class=Speech_C
  outputs []
  properties {
   softTaskIcon=&T-Icon-Replay-Plus-L
   generalConfig {defaultMessage=Denied: number entered was incorrect, defaultMetadataParameters [Message number {value=0}]}
  }
  sub_nodes [
   ConvoSideEffect_QuestPropActor_Message_C_20 {class=ConvoSideEffect_QuestPropActor_Message_C}
   ConversationChoiceNode_19 {class=ConversationChoiceNode, properties {defaultChoiceDisplayText=Enter 8}}
  ]
 }
 Speech_C_14 {
  id=#9
  class=Speech_C
  outputs [#12, #13, #14]
  properties {
   softTaskIcon=&T-Icon-Replay-Plus-L
   generalConfig {defaultMessage=072, defaultMetadataParameters [Message number {value=3}]}
  }
  sub_nodes [
   ConvoSideEffect_QuestPropActor_Message_C_21 {class=ConvoSideEffect_QuestPropActor_Message_C, properties {message Number=3}}
   ConversationChoiceNode_20 {class=ConversationChoiceNode, properties {defaultChoiceDisplayText=Enter 2}}
  ]
 }
 Speech_C_15 {
  id=#14
  class=Speech_C
  outputs [#15, #16, #17]
  properties {
   softTaskIcon=&T-Icon-Replay-Plus-L
   generalConfig {defaultMessage=0723, defaultMetadataParameters [Message number {value=4}]}
  }
  sub_nodes [
   ConvoSideEffect_QuestPropActor_Message_C_10 {class=ConvoSideEffect_QuestPropActor_Message_C, properties {message Number=4}}
   ConversationChoiceNode_9 {class=ConversationChoiceNode, properties {defaultChoiceDisplayText=Enter 3}}
  ]
 }
 Speech_C_19 {
  id=#17
  class=Speech_C
  outputs []
  properties {
   softTaskIcon=&T-Icon-Replay-Plus-L
   generalConfig {defaultMessage=Denied: number entered was incorrect, defaultMetadataParameters [Message number {value=0}]}
  }
  sub_nodes [
   ConvoSideEffect_QuestPropActor_Message_C_16 {class=ConvoSideEffect_QuestPropActor_Message_C}
   ConversationChoiceNode_15 {class=ConversationChoiceNode, properties {defaultChoiceDisplayText=Enter 4}}
  ]
 }
 Speech_C_20 {
  id=#16
  class=Speech_C
  outputs []
  properties {
   softTaskIcon=&T-Icon-Replay-Plus-L
   generalConfig {defaultMessage="Information downloaded!", defaultMetadataParameters [Message number {value=5}]}
  }
  sub_nodes [
   ConvoSideEffect_QuestPropActor_Message_C_17 {class=ConvoSideEffect_QuestPropActor_Message_C, properties {message Number=5}}
   ConversationChoiceNode_16 {class=ConversationChoiceNode, properties {defaultChoiceDisplayText=Enter 1}}
   ConvoSideEffect_QuestPropActor_ConvoSuccess_C_1 {class=ConvoSideEffect_QuestPropActor_ConvoSuccess_C}
  ]
 }
 Speech_C_21 {
  id=#15
  class=Speech_C
  outputs []
  properties {
   softTaskIcon=&T-Icon-Replay-Plus-L
   generalConfig {defaultMessage=Denied: number entered was incorrect, defaultMetadataParameters [Message number {value=0}]}
  }
  sub_nodes [
   ConvoSideEffect_QuestPropActor_Message_C_18 {class=ConvoSideEffect_QuestPropActor_Message_C}
   ConversationChoiceNode_17 {class=ConversationChoiceNode, properties {defaultChoiceDisplayText=Enter 3}}
  ]
 }
 Speech_C_16 {
  id=#13
  class=Speech_C
  outputs []
  properties {
   softTaskIcon=&T-Icon-Replay-Plus-L
   generalConfig {defaultMessage=Denied: number entered was incorrect, defaultMetadataParameters [Message number {value=0}]}
  }
  sub_nodes [
   ConvoSideEffect_QuestPropActor_Message_C_6 {class=ConvoSideEffect_QuestPropActor_Message_C}
   ConversationChoiceNode_3 {class=ConversationChoiceNode, properties {defaultChoiceDisplayText=Enter 6}}
  ]
 }
 Speech_C_2 {
  id=#12
  class=Speech_C
  outputs []
  properties {
   softTaskIcon=&T-Icon-Replay-Plus-L
   generalConfig {defaultMessage=Denied: number entered was incorrect, defaultMetadataParameters [Message number {value=0}]}
  }
  sub_nodes [
   ConvoSideEffect_QuestPropActor_Message_C_19 {class=ConvoSideEffect_QuestPropActor_Message_C}
   ConversationChoiceNode_18 {class=ConversationChoiceNode, properties {defaultChoiceDisplayText=Enter 7}}
  ]
 }
 Speech_C_17 {
  id=#7
  class=Speech_C
  outputs []
  properties {
   softTaskIcon=&T-Icon-Replay-Plus-L
   generalConfig {defaultMessage=Denied: number entered was incorrect, defaultMetadataParameters [Message number {value=0}]}
  }
  sub_nodes [
   ConvoSideEffect_QuestPropActor_Message_C_11 {class=ConvoSideEffect_QuestPropActor_Message_C}
   ConversationChoiceNode_8 {class=ConversationChoiceNode, properties {defaultChoiceDisplayText=Enter 2}}
  ]
 }
 Speech_C_12 {
  id=#6
  class=Speech_C
  outputs []
  properties {
   softTaskIcon=&T-Icon-Replay-Plus-L
   generalConfig {defaultMessage=Denied: number entered was incorrect, defaultMetadataParameters [Message number {value=0}]}
  }
  sub_nodes [
   ConvoSideEffect_QuestPropActor_Message_C_8 {class=ConvoSideEffect_QuestPropActor_Message_C}
   ConversationChoiceNode_6 {class=ConversationChoiceNode, properties {defaultChoiceDisplayText=Enter 3}}
  ]
 }
 Speech_C_1 {
  id=#3
  class=Speech_C
  outputs []
  properties {
   softTaskIcon=&T-Icon-Replay-Plus-L
   generalConfig {defaultMessage=Denied: number entered was incorrect, defaultMetadataParameters [Message number {value=0}]}
  }
  sub_nodes [
   ConvoSideEffect_QuestPropActor_Message_C_0 {class=ConvoSideEffect_QuestPropActor_Message_C}
   ConversationChoiceNode_1 {class=ConversationChoiceNode, properties {defaultChoiceDisplayText=Enter 8}}
  ]
 }
]
