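# Season 22 narrative quest conversation that works as a keypad prompt for access code 82133 on a quest prop actor.
# Each step offers three digit choices, where the correct one echoes the digits so far and leads onward, while either wrong one ends in a Denied message.
# Every speech node sends a message number to the prop via a side effect, tracking progress or zero for prompt and denial.
# The final correct digit shows "Information downloaded!" and fires the ConvoSuccess side effect.
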ConversationDatabase /BattlePassS22/Conversations/Convo_S22_P02_Q06_AccessCode_82133
entry_points [{tag=Conversation.Entry.Season22.Narrative.Code.82133, destinations [#1]}]
nodes [
 ConversationEntryPointNode_7 {
  id=#1
  class=ConversationEntryPointNode
  outputs [#2]
  properties {entryTag {tagName=Conversation.Entry.Season22.Narrative.Code.82133}}
 }
 Speech_C_46 {
  id=#2
  class=Speech_C
  outputs [#3, #4, #5]
  properties {
   softTaskIcon=/Game/UI/Foundation/Textures/Icons/Replay/T-Icon-Replay-Plus-L
   generalConfig {defaultMessage=Enter Valid Access Code:, defaultMetadataParameters [Message number {value=0}]}
  }
  sub_nodes [ConvoSideEffect_QuestPropActor_Message_C_46 {class=ConvoSideEffect_QuestPropActor_Message_C}]
 }
 Speech_C_42 {
  id=#5
  class=Speech_C
  outputs []
  properties {
   softTaskIcon=&T-Icon-Replay-Plus-L
   generalConfig {defaultMessage=Denied: number entered was incorrect, defaultMetadataParameters [Message number {value=0}]}
  }
  sub_nodes [
   ConvoSideEffect_QuestPropActor_Message_C_42 {class=ConvoSideEffect_QuestPropActor_Message_C}
   ConversationChoiceNode_39 {class=ConversationChoiceNode, properties {defaultChoiceDisplayText=Enter 5}}
  ]
 }
 Speech_C_47 {
  id=#4
  class=Speech_C
  outputs []
  properties {
   softTaskIcon=&T-Icon-Replay-Plus-L
   generalConfig {defaultMessage=Denied: number entered was incorrect, defaultMetadataParameters [Message number {value=0}]}
  }
  sub_nodes [
   ConvoSideEffect_QuestPropActor_Message_C_47 {class=ConvoSideEffect_QuestPropActor_Message_C}
   ConversationChoiceNode_43 {class=ConversationChoiceNode, properties {defaultChoiceDisplayText=Enter 0}}
  ]
 }
 Speech_C_41 {
  id=#3
  class=Speech_C
  outputs [#6, #7, #8]
  properties {
   softTaskIcon=&T-Icon-Replay-Plus-L
   generalConfig {defaultMessage=8, defaultMetadataParameters [Message number {value=1}]}
  }
  sub_nodes [
   ConvoSideEffect_QuestPropActor_Message_C_41 {class=ConvoSideEffect_QuestPropActor_Message_C, properties {message Number=1}}
   ConversationChoiceNode_38 {class=ConversationChoiceNode, properties {defaultChoiceDisplayText=Enter 8}}
  ]
 }
 Speech_C_43 {
  id=#8
  class=Speech_C
  outputs []
  properties {
   softTaskIcon=&T-Icon-Replay-Plus-L
   generalConfig {defaultMessage=Denied: number entered was incorrect, defaultMetadataParameters [Message number {value=0}]}
  }
  sub_nodes [
   ConvoSideEffect_QuestPropActor_Message_C_43 {class=ConvoSideEffect_QuestPropActor_Message_C}
   ConversationChoiceNode_40 {class=ConversationChoiceNode, properties {defaultChoiceDisplayText=Enter 7}}
  ]
 }
 Speech_C_44 {
  id=#7
  class=Speech_C
  outputs [#9, #10, #11]
  properties {
   softTaskIcon=&T-Icon-Replay-Plus-L
   generalConfig {defaultMessage=82, defaultMetadataParameters [Message number {value=2}]}
  }
  sub_nodes [
   ConvoSideEffect_QuestPropActor_Message_C_44 {class=ConvoSideEffect_QuestPropActor_Message_C, properties {message Number=2}}
   ConversationChoiceNode_41 {class=ConversationChoiceNode, properties {defaultChoiceDisplayText=Enter 2}}
  ]
 }
 Speech_C_38 {
  id=#11
  class=Speech_C
  outputs [#12, #13, #14]
  properties {
   softTaskIcon=&T-Icon-Replay-Plus-L
   generalConfig {defaultMessage=821, defaultMetadataParameters [Message number {value=3}]}
  }
  sub_nodes [
   ConvoSideEffect_QuestPropActor_Message_C_38 {class=ConvoSideEffect_QuestPropActor_Message_C, properties {message Number=3}}
   ConversationChoiceNode_35 {class=ConversationChoiceNode, properties {defaultChoiceDisplayText=Enter 1}}
  ]
 }
 Speech_C_35 {
  id=#14
  class=Speech_C
  outputs [#15, #16, #17]
  properties {
   softTaskIcon=&T-Icon-Replay-Plus-L
   generalConfig {defaultMessage=8213, defaultMetadataParameters [Message number {value=4}]}
  }
  sub_nodes [
   ConvoSideEffect_QuestPropActor_Message_C_35 {class=ConvoSideEffect_QuestPropActor_Message_C, properties {message Number=4}}
   ConversationChoiceNode_32 {class=ConversationChoiceNode, properties {defaultChoiceDisplayText=Enter 3}}
  ]
 }
 Speech_C_32 {
  id=#17
  class=Speech_C
  outputs []
  properties {
   softTaskIcon=&T-Icon-Replay-Plus-L
   generalConfig {defaultMessage=Denied: number entered was incorrect, defaultMetadataParameters [Message number {value=0}]}
  }
  sub_nodes [
   ConvoSideEffect_QuestPropActor_Message_C_32 {class=ConvoSideEffect_QuestPropActor_Message_C}
   ConversationChoiceNode_29 {class=ConversationChoiceNode, properties {defaultChoiceDisplayText=Enter 4}}
  ]
 }
 Speech_C_33 {
  id=#16
  class=Speech_C
  outputs []
  properties {
   softTaskIcon=&T-Icon-Replay-Plus-L
   generalConfig {defaultMessage=Denied: number entered was incorrect, defaultMetadataParameters [Message number {value=0}]}
  }
  sub_nodes [
   ConvoSideEffect_QuestPropActor_Message_C_33 {class=ConvoSideEffect_QuestPropActor_Message_C}
   ConversationChoiceNode_30 {class=ConversationChoiceNode, properties {defaultChoiceDisplayText=Enter 1}}
  ]
 }
 Speech_C_34 {
  id=#15
  class=Speech_C
  outputs []
  properties {
   softTaskIcon=&T-Icon-Replay-Plus-L
   generalConfig {defaultMessage="Information downloaded!", defaultMetadataParameters [Message number {value=5}]}
  }
  sub_nodes [
   ConvoSideEffect_QuestPropActor_Message_C_34 {class=ConvoSideEffect_QuestPropActor_Message_C, properties {message Number=5}}
   ConversationChoiceNode_31 {class=ConversationChoiceNode, properties {defaultChoiceDisplayText=Enter 3}}
   ConvoSideEffect_QuestPropActor_ConvoSuccess_C_1 {class=ConvoSideEffect_QuestPropActor_ConvoSuccess_C}
  ]
 }
 Speech_C_36 {
  id=#13
  class=Speech_C
  outputs []
  properties {
   softTaskIcon=&T-Icon-Replay-Plus-L
   generalConfig {defaultMessage=Denied: number entered was incorrect, defaultMetadataParameters [Message number {value=0}]}
  }
  sub_nodes [
   ConvoSideEffect_QuestPropActor_Message_C_36 {class=ConvoSideEffect_QuestPropActor_Message_C}
   ConversationChoiceNode_33 {class=ConversationChoiceNode, properties {defaultChoiceDisplayText=Enter 6}}
  ]
 }
 Speech_C_37 {
  id=#12
  class=Speech_C
  outputs []
  properties {
   softTaskIcon=&T-Icon-Replay-Plus-L
   generalConfig {defaultMessage=Denied: number entered was incorrect, defaultMetadataParameters [Message number {value=0}]}
  }
  sub_nodes [
   ConvoSideEffect_QuestPropActor_Message_C_37 {class=ConvoSideEffect_QuestPropActor_Message_C}
   ConversationChoiceNode_34 {class=ConversationChoiceNode, properties {defaultChoiceDisplayText=Enter 7}}
  ]
 }
 Speech_C_39 {
  id=#10
  class=Speech_C
  outputs []
  properties {
   softTaskIcon=&T-Icon-Replay-Plus-L
   generalConfig {defaultMessage=Denied: number entered was incorrect, defaultMetadataParameters [Message number {value=0}]}
  }
  sub_nodes [
   ConvoSideEffect_QuestPropActor_Message_C_39 {class=ConvoSideEffect_QuestPropActor_Message_C}
   ConversationChoiceNode_36 {class=ConversationChoiceNode, properties {defaultChoiceDisplayText=Enter 8}}
  ]
 }
 Speech_C_40 {
  id=#9
  class=Speech_C
  outputs []
  properties {
   softTaskIcon=&T-Icon-Replay-Plus-L
   generalConfig {defaultMessage=Denied: number entered was incorrect, defaultMetadataParameters [Message number {value=0}]}
  }
  sub_nodes [
   ConvoSideEffect_QuestPropActor_Message_C_40 {class=ConvoSideEffect_QuestPropActor_Message_C}
   ConversationChoiceNode_37 {class=ConversationChoiceNode, properties {defaultChoiceDisplayText=Enter 2}}
  ]
 }
 Speech_C_45 {
  id=#6
  class=Speech_C
  outputs []
  properties {
   softTaskIcon=&T-Icon-Replay-Plus-L
   generalConfig {defaultMessage=Denied: number entered was incorrect, defaultMetadataParameters [Message number {value=0}]}
  }
  sub_nodes [
   ConvoSideEffect_QuestPropActor_Message_C_45 {class=ConvoSideEffect_QuestPropActor_Message_C}
   ConversationChoiceNode_42 {class=ConversationChoiceNode, properties {defaultChoiceDisplayText=Enter 3}}
  ]
 }
]
