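# Shared NPC conversation for the Season 22 Battle Pass quest Quest_s22_Herald_Q08, where the player delivers The Herald's warning to seven NPCs.
# The shared quest entry point branches to one Speech choice per NPC, gated on the quest being active, that NPC's conversation entry tag, and its objective being incomplete.
# Each character-specific reply leads to a task that completes that NPC's objective by backend name, requiring the quest to turn in.
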ConversationDatabase /BattlePassS22/Conversations/Shared_Conversation_TheHerald
entry_points [{tag=Conversation.Entry.NPCLibrary.Quests, destinations [#1]}]
nodes [
 ConversationEntryPointNode_21 {
  id=#1
  class=ConversationEntryPointNode
  outputs [
   #2
   #3
   #4
   #5
   #6
   #7
   #8
  ]
  properties {entryTag {tagName=Conversation.Entry.NPCLibrary.Quests}}
 }
 Speech_C_71 {
  id=#8
  class=Speech_C
  outputs [#9]
  properties {
   softTaskIcon=&T_UI_NPC_MainQuest
   generalConfig {defaultMessage=I don't have time for threats. I have a business to run.}
  }
  sub_nodes [
   FortConversationRequirement_ControllerRequirement_241 {
    class=FortConversationRequirement_ControllerRequirement
    properties {
     nodeName=Require: Is Quest Active
     requirement {class=FortControllerRequirement_IsQuestActive, properties {questDefinition=/BattlePassS22/Items/QuestItems/Herald/Quest_s22_Herald_Q08}}
    }
   }
   FortConversationRequirement_ControllerRequirement_242 {
    class=FortConversationRequirement_ControllerRequirement
    properties {
     nodeName=Require: NPC Has Conversation Tag
     participantId {tagName=Conversation.Participant.NPC}
     requirement {class=FortControllerRequirement_NPCHasConversationEntryTag, properties {queryTag {tagName=Conversation.Entry.Guaco}}}
    }
   }
   ConversationChoiceNode_96 {
    class=ConversationChoiceNode
    properties {defaultChoiceDisplayText=Deliver The Herald's warning, choiceTags {gameplayTags [{tagName=Conversation.Choice.Category.Quest}]}}
   }
   FortConversationRequirement_ControllerRequirement_243 {
    class=FortConversationRequirement_ControllerRequirement
    properties {
     nodeName=Require: Has Completed Quest Objective
     requirement {
      class=FortControllerRequirement_HasCompletedQuestObjective
      properties {
       bInverseCondition=true
       objectiveBackendName=Quest_s22_Herald_Q08_obj12
       questDefinition=&Quest_s22_Herald_Q08
      }
     }
    }
   }
  ]
 }
 FortConversationTaskNode_CompleteQuestObjectiveByBackendName_68 {
  id=#9
  class=FortConversationTaskNode_CompleteQuestObjectiveByBackendName
  properties {objectiveBackendName=Quest_s22_Herald_Q08_obj12, questDefinition=&Quest_s22_Herald_Q08}
  sub_nodes [FortConversationRequirement_ControllerRequirement_244 {
    class=FortConversationRequirement_ControllerRequirement
    properties {
     nodeName=Require: Has Quest To Turn In
     requirement {class=FortControllerRequirement_HasQuestToTurnIn, properties {softQuestToTurnIn=&Quest_s22_Herald_Q08}}
    }
   }]
 }
 Speech_C_70 {
  id=#7
  class=Speech_C
  outputs [#10]
  properties {softTaskIcon=&T_UI_NPC_MainQuest, generalConfig {defaultMessage="I <b>foresaw</> this coming. I am prepared."}}
  sub_nodes [
   FortConversationRequirement_ControllerRequirement_267 {
    class=FortConversationRequirement_ControllerRequirement
    properties {
     nodeName=Require: Is Quest Active
     requirement {class=FortControllerRequirement_IsQuestActive, properties {questDefinition=&Quest_s22_Herald_Q08}}
    }
   }
   FortConversationRequirement_ControllerRequirement_268 {
    class=FortConversationRequirement_ControllerRequirement
    properties {
     nodeName=Require: NPC Has Conversation Tag
     participantId {tagName=Conversation.Participant.NPC}
     requirement {class=FortControllerRequirement_NPCHasConversationEntryTag, properties {queryTag {tagName=Conversation.Entry.Sunbird}}}
    }
   }
   ConversationChoiceNode_102 {
    class=ConversationChoiceNode
    properties {defaultChoiceDisplayText=Deliver The Herald's warning, choiceTags {gameplayTags [{tagName=Conversation.Choice.Category.Quest}]}}
   }
   FortConversationRequirement_ControllerRequirement_269 {
    class=FortConversationRequirement_ControllerRequirement
    properties {
     nodeName=Require: Has Completed Quest Objective
     requirement {
      class=FortControllerRequirement_HasCompletedQuestObjective
      properties {
       bInverseCondition=true
       objectiveBackendName=Quest_s22_Herald_Q08_obj11
       questDefinition=&Quest_s22_Herald_Q08
      }
     }
    }
   }
  ]
 }
 FortConversationTaskNode_CompleteQuestObjectiveByBackendName_67 {
  id=#10
  class=FortConversationTaskNode_CompleteQuestObjectiveByBackendName
  properties {objectiveBackendName=Quest_s22_Herald_Q08_obj11, questDefinition=&Quest_s22_Herald_Q08}
  sub_nodes [FortConversationRequirement_ControllerRequirement_270 {
    class=FortConversationRequirement_ControllerRequirement
    properties {
     nodeName=Require: Has Quest To Turn In
     requirement {class=FortControllerRequirement_HasQuestToTurnIn, properties {softQuestToTurnIn=&Quest_s22_Herald_Q08}}
    }
   }]
 }
 Speech_C_69 {
  id=#6
  class=Speech_C
  outputs [#11]
  properties {
   softTaskIcon=&T_UI_NPC_MainQuest
   generalConfig {defaultMessage=So this is how it happens... I need to send a transmission.}
  }
  sub_nodes [
   FortConversationRequirement_ControllerRequirement_263 {
    class=FortConversationRequirement_ControllerRequirement
    properties {
     nodeName=Require: Is Quest Active
     requirement {class=FortControllerRequirement_IsQuestActive, properties {questDefinition=&Quest_s22_Herald_Q08}}
    }
   }
   FortConversationRequirement_ControllerRequirement_264 {
    class=FortConversationRequirement_ControllerRequirement
    properties {
     nodeName=Require: NPC Has Conversation Tag
     participantId {tagName=Conversation.Participant.NPC}
     requirement {class=FortControllerRequirement_NPCHasConversationEntryTag, properties {queryTag {tagName=Conversation.Entry.Evie}}}
    }
   }
   ConversationChoiceNode_101 {
    class=ConversationChoiceNode
    properties {defaultChoiceDisplayText=Deliver The Herald's warning, choiceTags {gameplayTags [{tagName=Conversation.Choice.Category.Quest}]}}
   }
   FortConversationRequirement_ControllerRequirement_265 {
    class=FortConversationRequirement_ControllerRequirement
    properties {
     nodeName=Require: Has Completed Quest Objective
     requirement {
      class=FortControllerRequirement_HasCompletedQuestObjective
      properties {
       bInverseCondition=true
       objectiveBackendName=Quest_s22_Herald_Q08_obj10
       questDefinition=&Quest_s22_Herald_Q08
      }
     }
    }
   }
  ]
 }
 FortConversationTaskNode_CompleteQuestObjectiveByBackendName_66 {
  id=#11
  class=FortConversationTaskNode_CompleteQuestObjectiveByBackendName
  properties {objectiveBackendName=Quest_s22_Herald_Q08_obj10, questDefinition=&Quest_s22_Herald_Q08}
  sub_nodes [FortConversationRequirement_ControllerRequirement_266 {
    class=FortConversationRequirement_ControllerRequirement
    properties {
     nodeName=Require: Has Quest To Turn In
     requirement {class=FortControllerRequirement_HasQuestToTurnIn, properties {softQuestToTurnIn=&Quest_s22_Herald_Q08}}
    }
   }]
 }
 Speech_C_68 {
  id=#5
  class=Speech_C
  outputs [#12]
  properties {
   softTaskIcon=&T_UI_NPC_MainQuest
   generalConfig {defaultMessage=Yeah so threats from extradimensional beings aren't covered by your policy.}
  }
  sub_nodes [
   FortConversationRequirement_ControllerRequirement_259 {
    class=FortConversationRequirement_ControllerRequirement
    properties {
     nodeName=Require: Is Quest Active
     requirement {class=FortControllerRequirement_IsQuestActive, properties {questDefinition=&Quest_s22_Herald_Q08}}
    }
   }
   FortConversationRequirement_ControllerRequirement_260 {
    class=FortConversationRequirement_ControllerRequirement
    properties {
     nodeName=Require: NPC Has Conversation Tag
     participantId {tagName=Conversation.Participant.NPC}
     requirement {class=FortControllerRequirement_NPCHasConversationEntryTag, properties {queryTag {tagName=Conversation.Entry.TheUnderwriter}}}
    }
   }
   ConversationChoiceNode_100 {
    class=ConversationChoiceNode
    properties {defaultChoiceDisplayText=Deliver The Herald's warning, choiceTags {gameplayTags [{tagName=Conversation.Choice.Category.Quest}]}}
   }
   FortConversationRequirement_ControllerRequirement_261 {
    class=FortConversationRequirement_ControllerRequirement
    properties {
     nodeName=Require: Has Completed Quest Objective
     requirement {
      class=FortControllerRequirement_HasCompletedQuestObjective
      properties {
       bInverseCondition=true
       objectiveBackendName=Quest_s22_Herald_Q08_obj9
       questDefinition=&Quest_s22_Herald_Q08
      }
     }
    }
   }
  ]
 }
 FortConversationTaskNode_CompleteQuestObjectiveByBackendName_65 {
  id=#12
  class=FortConversationTaskNode_CompleteQuestObjectiveByBackendName
  properties {objectiveBackendName=Quest_s22_Herald_Q08_obj9, questDefinition=&Quest_s22_Herald_Q08}
  sub_nodes [FortConversationRequirement_ControllerRequirement_262 {
    class=FortConversationRequirement_ControllerRequirement
    properties {
     nodeName=Require: Has Quest To Turn In
     requirement {class=FortControllerRequirement_HasQuestToTurnIn, properties {softQuestToTurnIn=&Quest_s22_Herald_Q08}}
    }
   }]
 }
 Speech_C_67 {
  id=#4
  class=Speech_C
  outputs [#13]
  properties {
   softTaskIcon=&T_UI_NPC_MainQuest
   generalConfig {defaultMessage="The Herald's gonna do <b>WHAT</> now? Oh wow. This is my time to shine."}
  }
  sub_nodes [
   FortConversationRequirement_ControllerRequirement_255 {
    class=FortConversationRequirement_ControllerRequirement
    properties {
     nodeName=Require: Is Quest Active
     requirement {class=FortControllerRequirement_IsQuestActive, properties {questDefinition=&Quest_s22_Herald_Q08}}
    }
   }
   FortConversationRequirement_ControllerRequirement_256 {
    class=FortConversationRequirement_ControllerRequirement
    properties {
     nodeName=Require: NPC Has Conversation Tag
     participantId {tagName=Conversation.Participant.NPC}
     requirement {class=FortControllerRequirement_NPCHasConversationEntryTag, properties {queryTag {tagName=Conversation.Entry.Panther}}}
    }
   }
   ConversationChoiceNode_99 {
    class=ConversationChoiceNode
    properties {defaultChoiceDisplayText=Deliver The Herald's warning, choiceTags {gameplayTags [{tagName=Conversation.Choice.Category.Quest}]}}
   }
   FortConversationRequirement_ControllerRequirement_257 {
    class=FortConversationRequirement_ControllerRequirement
    properties {
     nodeName=Require: Has Completed Quest Objective
     requirement {
      class=FortControllerRequirement_HasCompletedQuestObjective
      properties {
       bInverseCondition=true
       objectiveBackendName=Quest_s22_Herald_Q08_obj4
       questDefinition=&Quest_s22_Herald_Q08
      }
     }
    }
   }
  ]
 }
 FortConversationTaskNode_CompleteQuestObjectiveByBackendName_64 {
  id=#13
  class=FortConversationTaskNode_CompleteQuestObjectiveByBackendName
  properties {objectiveBackendName=Quest_s22_Herald_Q08_obj4, questDefinition=&Quest_s22_Herald_Q08}
  sub_nodes [FortConversationRequirement_ControllerRequirement_258 {
    class=FortConversationRequirement_ControllerRequirement
    properties {
     nodeName=Require: Has Quest To Turn In
     requirement {class=FortControllerRequirement_HasQuestToTurnIn, properties {softQuestToTurnIn=&Quest_s22_Herald_Q08}}
    }
   }]
 }
 Speech_C_66 {
  id=#3
  class=Speech_C
  outputs [#14]
  properties {softTaskIcon=&T_UI_NPC_MainQuest, generalConfig {defaultMessage=I'm not really much of a reader but thanks.}}
  sub_nodes [
   FortConversationRequirement_ControllerRequirement_251 {
    class=FortConversationRequirement_ControllerRequirement
    properties {
     nodeName=Require: Is Quest Active
     requirement {class=FortControllerRequirement_IsQuestActive, properties {questDefinition=&Quest_s22_Herald_Q08}}
    }
   }
   FortConversationRequirement_ControllerRequirement_252 {
    class=FortConversationRequirement_ControllerRequirement
    properties {
     nodeName=Require: NPC Has Conversation Tag
     participantId {tagName=Conversation.Participant.NPC}
     requirement {class=FortControllerRequirement_NPCHasConversationEntryTag, properties {queryTag {tagName=Conversation.Entry.Troops}}}
    }
   }
   ConversationChoiceNode_98 {
    class=ConversationChoiceNode
    properties {defaultChoiceDisplayText=Deliver The Herald's warning, choiceTags {gameplayTags [{tagName=Conversation.Choice.Category.Quest}]}}
   }
   FortConversationRequirement_ControllerRequirement_253 {
    class=FortConversationRequirement_ControllerRequirement
    properties {
     nodeName=Require: Has Completed Quest Objective
     requirement {
      class=FortControllerRequirement_HasCompletedQuestObjective
      properties {
       bInverseCondition=true
       objectiveBackendName=Quest_s22_Herald_Q08_obj2
       questDefinition=&Quest_s22_Herald_Q08
      }
     }
    }
   }
  ]
 }
 FortConversationTaskNode_CompleteQuestObjectiveByBackendName_63 {
  id=#14
  class=FortConversationTaskNode_CompleteQuestObjectiveByBackendName
  properties {objectiveBackendName=Quest_s22_Herald_Q08_obj2, questDefinition=&Quest_s22_Herald_Q08}
  sub_nodes [FortConversationRequirement_ControllerRequirement_254 {
    class=FortConversationRequirement_ControllerRequirement
    properties {
     nodeName=Require: Has Quest To Turn In
     requirement {class=FortControllerRequirement_HasQuestToTurnIn, properties {softQuestToTurnIn=&Quest_s22_Herald_Q08}}
    }
   }]
 }
 Speech_C_51 {
  id=#2
  class=Speech_C
  outputs [#15]
  properties {
   softTaskIcon=&T_UI_NPC_MainQuest
   generalConfig {defaultMessage=I think it's safe to say everyone has a right to be scared.}
  }
  sub_nodes [
   FortConversationRequirement_ControllerRequirement_246 {
    class=FortConversationRequirement_ControllerRequirement
    properties {
     nodeName=Require: Is Quest Active
     requirement {class=FortControllerRequirement_IsQuestActive, properties {questDefinition=&Quest_s22_Herald_Q08}}
    }
   }
   FortConversationRequirement_ControllerRequirement_247 {
    class=FortConversationRequirement_ControllerRequirement
    properties {
     nodeName=Require: NPC Has Conversation Tag
     participantId {tagName=Conversation.Participant.NPC}
     requirement {class=FortControllerRequirement_NPCHasConversationEntryTag, properties {queryTag {tagName=Conversation.Entry.Cryptic}}}
    }
   }
   ConversationChoiceNode_97 {
    class=ConversationChoiceNode
    properties {defaultChoiceDisplayText=Deliver The Herald's Warning, choiceTags {gameplayTags [{tagName=Conversation.Choice.Category.Quest}]}}
   }
   FortConversationRequirement_ControllerRequirement_248 {
    class=FortConversationRequirement_ControllerRequirement
    properties {
     nodeName=Require: Has Completed Quest Objective
     requirement {
      class=FortControllerRequirement_HasCompletedQuestObjective
      properties {
       bInverseCondition=true
       objectiveBackendName=Quest_s22_Herald_Q08_obj1
       questDefinition=&Quest_s22_Herald_Q08
      }
     }
    }
   }
  ]
 }
 FortConversationTaskNode_CompleteQuestObjectiveByBackendName_44 {
  id=#15
  class=FortConversationTaskNode_CompleteQuestObjectiveByBackendName
  properties {objectiveBackendName=Quest_s22_Herald_Q08_obj1, questDefinition=&Quest_s22_Herald_Q08}
  sub_nodes [FortConversationRequirement_ControllerRequirement_249 {
    class=FortConversationRequirement_ControllerRequirement
    properties {
     nodeName=Require: Has Quest To Turn In
     requirement {class=FortControllerRequirement_HasQuestToTurnIn, properties {softQuestToTurnIn=&Quest_s22_Herald_Q08}}
    }
   }]
 }
]
